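# Master translucent, unlit, two-sided material for vehicle boost elements on Niagara ribbons and mesh particles, with fire and water variants.
# Panning textures, particle color and a dissolve-shape fire tint build emissive; gradient masks, panned textures and normal-distorted caustics build opacity.
# Scalar-compare If nodes select Fire/Water and Ribbon/Ring behavior without static permutations.
# Optional TrailBend drives sine/cosine vertex waves along UV1, scaled by the WPO dynamic parameter.
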Material /VehicleCosmetics/Boosts/Boost_Elements/Materials/NewMaterials_Opt/M_Boost_MasterElements
graph {
 properties {
  bUsedWithNiagaraMeshParticles=true
  bUsedWithNiagaraRibbons=true
  blendMode=BLEND_Translucent
  pixelDepthOffsetMode=PDOM_Legacy
  shadingModel=MSM_Unlit
  twoSided=true
  usedShadingModels=Unlit
 }
 nodes [
  n1 = ParticleColor()
  n2 = ComponentMask(None=TextureCoordinate() {coordinateIndex=1}) {r=true}
  n3 = OneMinus(None=@n2)
  n4 = Multiply(A=Multiply(A=@n2, B=@n3), B=Constant() {r=4})
  n5 = Multiply(
   A=TextureCoordinate() {uTiling=4, vTiling=0.4}
   B=MaterialFunctionCall(
    X=ScalarParameter() {defaultValue=1, group=Textures, parameterName=Tiling X T4}
    Y=ScalarParameter() {defaultValue=1, group=Textures, parameterName=Tiling Y T4}
   ) {materialFunction=&MakeFloat2}.Result
  )
  n6 = MaterialFunctionCall(
   X=ScalarParameter() {defaultValue=-5, group=Textures, parameterName=Speed X T4}
   Y=ScalarParameter() {defaultValue=0.4, group=Textures, parameterName=Speed Y T4}
  ) {materialFunction=&MakeFloat2}
  n7 = Multiply(
   A=Power(Base=@n4, Exp=Constant() {r=3})
   B=TextureSampleParameter2D(UVs=Panner(Coordinate=@n5, Speed=@n6.Result) {speedX=-5, speedY=0.4}) {parameterName=DissolveShape, texture=/DealershipSharedLibrary/FX/Textures/Gradient_Masks/T_VCVFX_Shapes_A}.RGB
  )
  n8 = Add(A=Multiply(A=@n7, B=Constant() {r=2}), B=@n7)
  n9 = Multiply(
   A=If(
    A=ScalarParameter() {parameterName=Fire}
    B=ScalarParameter() {defaultValue=1, parameterName=Water}
    "A > B"=Constant3Vector() {constant=<rgba 5 0.05 0 0>}
    "A < B"=Constant()
   )
   B=@n8
  )
  n10 = TextureCoordinate() {uTiling=3, vTiling=0.5}
  n11 = MaterialFunctionCall(
   X=ScalarParameter() {defaultValue=1, group=Textures, parameterName=Tiling X T1}
   Y=ScalarParameter() {defaultValue=1, group=Textures, parameterName=Tiling Y T1}
  ) {materialFunction=&MakeFloat2}
  n12 = MaterialFunctionCall(X=ScalarParameter() {defaultValue=-4, group=Textures, parameterName=Speed X T1/T2}, Y=Constant()) {materialFunction=&MakeFloat2}
  n13 = TextureSampleParameter2D(UVs=Panner(Coordinate=Multiply(A=@n10, B=@n11.Result), Speed=@n12.Result) {speedX=-4}) {parameterName=Texture_1, texture=/DealershipSharedLibrary/FX/Textures/TexturePackage/T_VCVFX_WaterTexture2_RGB}
  n14 = VectorParameter() {parameterName=Color_Fire_Water, defaultValue=<rgba 0.01 0.01 0.01 1>}
  MaterialFunctionCall(Emissive=Add(A=Multiply(A=@n13.B, B=@n14.RGB), B=Add(A=@n1.RGB, B=@n9))) {materialFunction=&MF_FX_EmissiveAdjustments}
  n15 = TextureCoordinate()
  n16 = MaterialFunctionCall(X=ScalarParameter() {group=Textures, parameterName=Offset Gradient}, Y=Constant()) {materialFunction=&MakeFloat2}
  n17 = ScalarParameter() {defaultValue=1, group=Opacity, parameterName=Contrast Fade Front, sortPriority=5}
  n18 = ScalarParameter() {defaultValue=1, group=Opacity, parameterName=Brightness Fade Front, sortPriority=6}
  n19 = Constant() {r=1}
  n20 = ScalarParameter() {defaultValue=1, group=Opacity, parameterName=Contrast Fade Back, sortPriority=7}
  n21 = ScalarParameter() {defaultValue=1, group=Opacity, parameterName=Brightness Fade Back, sortPriority=8}
  n22 = StaticSwitchParameter(
   True=TextureSample(UVs=Add(A=@n15, B=@n16.Result)) {texture=/DealershipSharedLibrary/FX/Textures/Gradient_Masks/T_VCVFX_SingularityTrailTransparency_A}.RGB
   False=Multiply(
    A=MaterialFunctionCall!(
     "Gradient U=0 V=1"
     "Invert gradient 0=False 1=True"
     "Contrast (Power)"=@n17
     "Brightness (Multiply)"=@n18
     "Dynamic Contrast (Power)"
     "Dynamic Brightness (Multiply)"
    ) {materialFunction=/DealershipSharedLibrary/FX/Materials/MaterialFunctions/MF_LinearGradientPlus}.Result
    B=MaterialFunctionCall("Invert gradient 0=False 1=True"=@n19, "Contrast (Power)"=@n20, "Brightness (Multiply)"=@n21) {materialFunction=&MF_LinearGradientPlus}.Result
   )
  ) {defaultValue=true, parameterName=UseTexture gradient?}
  n23 = Constant() {r=1.5}
  n24 = If(
   A=ScalarParameter() {defaultValue=1, parameterName=Ribbon?}
   B=ScalarParameter() {parameterName=Ring?}
   "A > B"=Power(Base=Multiply(A=@n22, B=@n23), Exp=ScalarParameter() {defaultValue=2, group=Textures, parameterName=Contrast Text Gradient})
   "A < B"=Constant() {r=1}
  )
  n25 = TextureCoordinate() {uTiling=3}
  n26 = MaterialFunctionCall(
   X=ScalarParameter() {defaultValue=1, group=Textures, parameterName=Tiling X T2}
   Y=ScalarParameter() {defaultValue=1, group=Textures, parameterName=Tiling Y T2}
  ) {materialFunction=&MakeFloat2}
  n27 = TextureSampleParameter2D(UVs=Panner(Coordinate=Multiply(A=@n25, B=@n26.Result), Speed=@n12.Result) {speedX=-4}) {parameterName=Texture_2, texture=&T_VCVFX_WaterTexture2_RGB}
  n28 = ScalarParameter() {defaultValue=1, group=Textures, parameterName=Tiling X T3}
  n29 = ScalarParameter() {defaultValue=1, group=Textures, parameterName=Tiling Y T3}
  n30 = Panner(
   Coordinate=Multiply(
    A=TextureCoordinate() {uTiling=2, vTiling=0.5}
    B=MaterialFunctionCall(X=@n28, Y=@n29) {materialFunction=&MakeFloat2}.Result
   )
   Speed=MaterialFunctionCall(
    X=ScalarParameter() {defaultValue=-7, group=Textures, parameterName=Speed X T3}
    Y=ScalarParameter() {defaultValue=0.2, group=Textures, parameterName=Speed Y T3}
   ) {materialFunction=&MakeFloat2}.Result
  ) {speedX=-7, speedY=0.2}
  n31 = Add(
   A=Multiply(A=Multiply(A=@n27.G) {constB=0.5}, B=Add(A=@n24, B=@n8))
   B=If(
    A=ScalarParameter() {parameterName=Fire}
    B=ScalarParameter() {defaultValue=1, parameterName=Water}
    "A > B"=TextureSampleParameter2D(UVs=@n30) {parameterName=Texture_ExtraFire, texture=/DealershipSharedLibrary/FX/Textures/TexturePackage/T_VCVFX_FireTexture_RGB}.R
    "A < B"=Constant()
   )
  )
  n32 = TextureCoordinate() {uTiling=0.5, vTiling=0.5}
  n33 = Constant() {r=2}
  n34 = Time()
  n35 = Constant() {r=1}
  n36 = Multiply(A=@n33, B=TextureCoordinate() {uTiling=0.1, vTiling=0.1})
  n37 = Add(
   A=TextureSample(UVs=Panner(Coordinate=@n36) {speedX=0.05}) {samplerType=SAMPLERTYPE_Normal, texture=/DealershipSharedLibrary/FX/Textures/Normal/T_VCVFX_Water03Swizzled_Noise}.RGB
   B=TextureSample(UVs=Panner(Coordinate=@n36) {speedX=-0.05, speedY=0.05}) {samplerType=SAMPLERTYPE_Normal, texture=&T_VCVFX_Water03Swizzled_Noise}.RGB
  )
  n38 = Add(
   A=Panner(Coordinate=Multiply(A=@n32, B=@n33), Time=Multiply(A=@n34, B=@n35)) {speedX=-0.5, speedY=-0.5}
   B=Multiply(A=ComponentMask(None=@n37) {g=true, r=true}, B=Constant() {r=0.3})
  )
  n39 = Multiply(
   A=@n31
   B=Multiply(
    A=TextureSample(UVs=@n38) {texture=/DealershipSharedLibrary/FX/Textures/Noise/T_VCVFX_Caustics_A}.RGB
    B=ScalarParameter() {defaultValue=8, parameterName=ForceWater}
   ) {constB=0}
  )
  Multiply(A=@n1.A, B=Saturate(None=Add(A=@n31, B=@n39)))
  n40 = Constant() {r=0.3}
  n41 = ComponentMask(None=TextureCoordinate() {coordinateIndex=1, uTiling=1.5}) {r=true}
  n42 = Multiply(A=Add(A=@n3, B=@n40), B=Add(A=@n41, B=@n40))
  n43 = ScalarParameter() {defaultValue=1, parameterName=WPO_Ribbon}
  n44 = ScalarParameter() {parameterName=WPO_Ring}
  n45 = OneMinus(None=Multiply(A=Time(), B=Constant() {r=3}))
  n46 = Cosine(None=Add(A=@n45, B=@n41)) {period=2}
  n47 = Constant() {r=1}
  n48 = Sine(None=Add(A=@n41, B=Multiply(A=@n47, B=@n45))) {period=2}
  n49 = AppendVector(
   A=StaticSwitchParameter(True=AppendVector(A=@n46, B=@n48), False=AppendVector(A=@n48, B=@n46)) {defaultValue=true, parameterName=invert}
   B=StaticSwitchParameter(True=@n48, False=@n46) {defaultValue=true, parameterName=invert}
  )
  n50 = ScalarParameter() {defaultValue=11, parameterName=Offset}
  n51 = Transform(None=Multiply(A=Clamp(None=@n49) {maxDefault=2, minDefault=-2}, B=@n50)) {transformType=TRANSFORM_Tangent}
  n52 = Transform(None=Multiply(A=Saturate(None=@n49), B=@n50))
  n53 = Multiply(
   A=If(A=ScalarParameter() {defaultValue=1, parameterName=WPO_Ribbon}, B=ScalarParameter() {parameterName=WPO_Ring}, "A > B"=@n42, "A < B"=@n4)
   B=Multiply(A=If(A=@n43, B=@n44, "A > B"=@n51, "A < B"=@n52), B=@n47)
  )
  n54 = DynamicParameter() {defaultValue=<rgba 1 0 1 1>, paramNames [WPO,Fire,Water,P]}
  Reroute(None=StaticSwitchParameter(True=Multiply(A=@n53, B=@n54.WPO), False=Constant()) {defaultValue=true, group=TrailBend, parameterName=TrailBend})
  Reroute(None=@n13.B)
  Reroute(None=@n51)
  Reroute(None=@n42)
 ]
}
